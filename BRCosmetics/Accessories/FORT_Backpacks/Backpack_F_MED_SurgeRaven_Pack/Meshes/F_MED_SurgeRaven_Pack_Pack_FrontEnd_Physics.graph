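# Physics asset for the front-end version of the SurgeRaven backpack mesh, simulating a dangling cord on the pack.
# A six-segment chain of small capsules hangs from a base cord body, linked by swing limits and tight twist limits, with the widest swing at the root.
# The base cord joins a large pack capsule through a constraint with free swing and twist.
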PhysicsAsset /BRCosmetics/Accessories/FORT_Backpacks/Backpack_F_MED_SurgeRaven_Pack/Meshes/F_MED_SurgeRaven_Pack_Pack_FrontEnd_Physics
bodies {
 base_cord [{shapeType=Capsule, length=0.505, radius=0.505, shapeName=base_cord_capsule, rotation=<rot 0 -90 0>}]
 dyn_cord_01 [{shapeType=Capsule, length=1.8566, radius=0.4, shapeName=dyn_cord_01_capsule, center=<1.43 0 0>, rotation=<rot 90 0 0>}]
 dyn_cord_02 [{shapeType=Capsule, length=2.0642, radius=0.4, shapeName=dyn_cord_02_capsule, center=<1.2556 0 0>, rotation=<rot 90 0 0>}]
 dyn_cord_03 [{shapeType=Capsule, length=2.9123, radius=0.5, shapeName=dyn_cord_03_capsule, center=<1.9415 0 0>, rotation=<rot 90 0 0>}]
 dyn_cord_04 [{shapeType=Capsule, length=2.505, radius=0.505, shapeName=dyn_cord_04_capsule, center=<1.6969 0 0>, rotation=<rot 90 0 0>}]
 dyn_cord_05 [{shapeType=Capsule, length=0.9855, radius=0.6976, shapeName=dyn_cord_05_capsule, center=<1.2884 0 0>, rotation=<rot 90 0 0>}]
 dyn_cord_06 [{shapeType=Capsule, length=0.505, radius=0.505, shapeName=dyn_cord_06_capsule, center=<0.9355 0.1062 0.0161>, rotation=<rot 90 0 0>}]
 pack [{shapeType=Capsule, length=36.974, radius=5.4046, center=<-4.5868 1.9166 -2.7455>, rotation=<rot 0.1693 -90 -38.0865>}]
}
constraints [
 {bone1Name=dyn_cord_01, bone2Name=base_cord, swing1LimitDegrees=80, swing2LimitDegrees=30, twistLimitDegrees=5}
 {bone1Name=dyn_cord_02, bone2Name=dyn_cord_01, swing1LimitDegrees=20, swing2LimitDegrees=10, twistLimitDegrees=5}
 {bone1Name=dyn_cord_03, bone2Name=dyn_cord_02, swing1LimitDegrees=40, swing2LimitDegrees=20, twistLimitDegrees=5}
 {bone1Name=dyn_cord_04, bone2Name=dyn_cord_03, swing1LimitDegrees=40, swing2LimitDegrees=20, twistLimitDegrees=5}
 {bone1Name=dyn_cord_05, bone2Name=dyn_cord_04, swing1LimitDegrees=20, swing2LimitDegrees=20, twistLimitDegrees=5}
 {
  bone1Name=base_cord
  bone2Name=pack
  swing1LimitDegrees=45
  swing1Motion=Free
  swing2LimitDegrees=45
  swing2Motion=Free
  twistLimitDegrees=45
  twistMotion=Free
 }
 {bone1Name=dyn_cord_06, bone2Name=dyn_cord_05, swing1LimitDegrees=20, swing2LimitDegrees=20, twistLimitDegrees=5}
]
