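# Animation Blueprint for the SurgeRaven back bling, adding secondary motion to the pack and its cord.
# An offset pack pose feeds front-end or in-game RigidBody simulation, with a planar-limited trail controller on the cord when rigid bodies are disabled.
# The event graph sets gravity by skydive/front-end state and flags SunBurst emote montages.
# It also hides a material element while the matching pickaxe is equipped, except during skydiving.
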AnimBlueprint /BRCosmetics/Accessories/FORT_Backpacks/Backpack_F_MED_SurgeRaven_Pack/Meshes/F_MED_SurgeRaven_Pack_Pack_AnimBP
target_skeleton=&/F_MED_SurgeRaven_Pack_Pack_Skeleton
graphs [
 AnimGraph {nodes [
   n1 = RigidBody!(ComponentPose=Local To Component(LocalPose=Use cached pose 'Initial Pose'()), Alpha, OverrideWorldGravity:Vector="0,0,0", ExternalForce) -> (Pose) {
    bClampLinearTranslationLimitToRefPose=true
    bOverrideWorldGravity=true
    overridePhysicsAsset=&/F_MED_SurgeRaven_Pack_Pack_Physics
    simulationSpace=BaseBoneSpace
    alphaBoolBlend {blendOption=Linear}
    baseBoneRef {boneName=pack}
    componentAppliedLinearAccClamp=<0 0 0>
    simSpaceSettings {
     maxAngularAcceleration=40
     maxAngularVelocity=8
     maxLinearAcceleration=3500
     maxLinearVelocity=700
     worldAlpha=0.425
     externalLinearDragV=<1 1 1>
    }
   }
   n2 = Component To Local(ComponentPose=@n1)
   n3 = RigidBody(ComponentPose=Local To Component(LocalPose=Use cached pose 'Initial Pose'()), OverrideWorldGravity="0,0,0") {
    bClampLinearTranslationLimitToRefPose=true
    bOverrideWorldGravity=true
    overridePhysicsAsset=&/F_MED_SurgeRaven_Pack_Pack_FrontEnd_Physics
    simulationSpace=BaseBoneSpace
    alphaBoolBlend {blendOption=Linear}
    baseBoneRef {boneName=pack}
    componentAppliedLinearAccClamp=<0 0 0>
    simSpaceSettings {
     maxAngularAcceleration=40
     maxAngularVelocity=8
     maxLinearAcceleration=3500
     maxLinearVelocity=700
     worldAlpha=1
     externalLinearDragV=<1 1 1>
    }
   }
   n4 = Component To Local(ComponentPose=@n3)
   n5 = Get bIsInFrontEnd()
   n6 = "Transform (Modify) Bone - Bone: base_cord"!(
    ComponentPose=Local To Component(LocalPose=Use cached pose 'Initial Pose'())
    bAlphaBoolEnabled:bool=True
    Translation="0,0,0"
    Rotation="0,80,0"
    Scale="1,1,1"
   ) -> (Pose) {
    alphaInputType=Bool
    rotationMode=BMM_Additive
    rotationSpace=BCS_BoneSpace
    translationMode=BMM_Additive
    translationSpace=BCS_BoneSpace
    alphaBoolBlend {blendInTime=1, blendOutTime=1}
    boneToModify {boneName=base_cord}
    rotation {yaw=80}
   }
   n7 = Trail controller - Bone: dyn_cord_06(ComponentPose=@n6, Alpha=0.6) {
    bLimitRotation=true
    bShowLimit=false
    bUsePlanarLimit=true
    chainLength=6
    debugLifeTime=0.1
    maxDeltaTime=0.0333
    relaxationSpeedScale=2
    alphaScaleBiasClamp {bInterpResult=true}
    baseJoint {boneName=root}
    planarLimits [{drivingBone {boneName=pack}, planeTransform {location=<0 0 0>, rotation=<rot 0 0 -90>, scale=<1 1 1>}}]
    rotationLimits [
     {limitMax=<0 0 10>, limitMin=<0 -5 -10>}
     {limitMax=<0 0 0>, limitMin=<0 0 -20>}
     {limitMax=<0 0 20>, limitMin=<0 0 -20>}
     {limitMax=<0 0 20>, limitMin=<0 0 -20>}
     {limitMax=<0 0 20>, limitMin=<0 0 -20>}
     {limitMax=<0 0 20>, limitMin=<0 -5 -20>}
    ]
    rotationOffsets [<0 0 0>, <0 0 0>, <0 0 0>, <0 0 0>, <0 0 0>, <0 0 0>]
    trailBone {boneName=dyn_cord_06}
    trailRelaxationSpeed {editorCurveData {keys [<at 0 23 RCIM_Linear>,<at 0.284 19.3799 RCIM_Linear>,<at 0.7722 8.3926 RCIM_Linear>,<at 1 5 RCIM_Linear>]}}
   }
   Output Pose(Result=Blend Poses by bool(
     BlendPose_0=Blend Poses by bool(BlendPose_0=@n4, BlendPose_1=@n2, bActiveValue=@n5)
     BlendPose_1=Component To Local(ComponentPose=@n7)
     bActiveValue=Get bRigidBodyEnabled()
    )) {layerGroup=None}
   n8 = "Transform (Modify) Bone - Bone: pack"!(ComponentPose=Local To Component(LocalPose=Input Pose()), bAlphaBoolEnabled:bool, Translation="5,0,0", Rotation="-10,0,0", Scale) -> (Pose) {
    alphaInputType=Bool
    rotationMode=BMM_Additive
    rotationSpace=BCS_BoneSpace
    translationMode=BMM_Additive
    translationSpace=BCS_BoneSpace
    alphaBoolBlend {blendInTime=1, blendOutTime=1}
    boneToModify {boneName=pack}
    rotation {pitch=-10}
    translation {x=5}
   }
   Save cached pose 'Initial Pose'(Pose=Component To Local(ComponentPose=@n8)) {cacheName=Initial Pose}
  ]}
 EventGraph {nodes [
   Event BlueprintUpdateAnimation()
   > n1 = Sequence()
   > Set GravityOverride(GravityOverride="0, 0, 0")
   > n2 = Branch(Condition=Get bIsSkydiveFloating())
   > Set GravityOverride(GravityOverride="0,0,550")
   > Branch(Condition=Get bIsInFrontEnd())
   > Set GravityOverride(GravityOverride="0,0,200")
   Set GravityOverride(execute=@n2.else, GravityOverride="0,0,-980")
   n3 = GetAnimInstance(self=GetOwnerSkeletalMeshComponent())
   Is Valid(exec=@n1.then_1, InputObject=@n3)
   > Set Is playing StayAfloat(Is playing StayAfloat:bool=OR Boolean!(
     A=Montage_IsPlaying(self=@n3, Montage=/BRCosmetics/Animation/Game/MainPlayer/Emotes/SunBurst/Emote_SunBurst_CLF_M)
     B=Montage_IsPlaying(self=@n3, Montage=/BRCosmetics/Animation/Game/MainPlayer/Emotes/SunBurst/Emote_SunBurst_CLF_M_Frontend)
     C:bool=Montage_IsPlaying(self=@n3, Montage=/BRCosmetics/Animation/Game/MainPlayer/Emotes/SunBurst/Emote_SunBurst_CLF_Skirt_Frontend)
     D:bool=Montage_IsPlaying(self=@n3, Montage=/BRCosmetics/Animation/Game/MainPlayer/Emotes/SunBurst/Emote_SunBurst_CLF_Skirt_M)
     E:bool=Montage_IsPlaying(self=@n3, Montage=/BRCosmetics/Animation/Game/MainPlayer/Emotes/SunBurst/Emote_SunBurst_CLM_M)
     F:bool=Montage_IsPlaying(self=@n3, Montage=/BRCosmetics/Animation/Game/MainPlayer/Emotes/SunBurst/Emote_SunBurst_CLM_M_Frontend)
     G:bool=Montage_IsPlaying(self=@n3, Montage=/BRCosmetics/Animation/Game/MainPlayer/Emotes/SunBurst/Emote_SunBurst_CMF_M)
     H:bool=Montage_IsPlaying(self=@n3, Montage=/BRCosmetics/Animation/Game/MainPlayer/Emotes/SunBurst/Emote_SunBurst_CMF_M_Frontend)
     I:bool=Montage_IsPlaying(self=@n3, Montage=/BRCosmetics/Animation/Game/MainPlayer/Emotes/SunBurst/Emote_SunBurst_CMF_Skirt_M)
     J:bool=Montage_IsPlaying(self=@n3, Montage=/BRCosmetics/Animation/Game/MainPlayer/Emotes/SunBurst/Emote_SunBurst_CMF_Skirt_M_Frontend)
     K:bool=Montage_IsPlaying(self=@n3, Montage=/BRCosmetics/Animation/Game/MainPlayer/Emotes/SunBurst/Emote_SunBurst_CMM_M)
     L:bool=Montage_IsPlaying(self=@n3, Montage=/BRCosmetics/Animation/Game/MainPlayer/Emotes/SunBurst/Emote_SunBurst_CMM_M_Frontend)
    ) -> (ReturnValue)) -> (Output_Get:bool)
   n4 = Event OnWeaponChanged()
   n5 = Get Dynamic Material Instance(execute=[@n4.then, @n11.then], SkeletalMeshComponent=GetPartSkeletalMeshComponent())
   n6 = PickaxeItemDefinitionHasAnyMetaTag(Tags=Make Array("[0]"=(TagName="Cosmetics.Metatags.SurgeRaven_Pack")), bRequiresEquipped=true)
   > n7 = Branch(Condition=@n6)
   > n8 = Branch(Condition=Get bIsSkyDiving())
   SetScalarParameterValue(execute=[@n9.then, @n8.then, @n10.else, @n7.else], self=@n5, ParameterName=Hide Element 01, Value=0)
   SetScalarParameterValue(execute=[@n8.else, @n10.then], self=@n5, ParameterName=Hide Element 01, Value=1)
   n9 = Event OnBeginSkydiving()
   Event OnEndSkydiving()
   > n10 = Branch(Condition=@n6)
   n11 = Event OnCharacterCustomizationLoaded()
   TryGetPawnOwner()
  ]}
]
variables [
 Gravity Override Params {
  type=FortGravityOverrideParameterSet
  default=(Skydive_GoingUp=(GravityOverride=()),Skydive_FloatingDown=(GravityOverride=()),Skydive_DivingDown=(GravityOverride=()),Skydive_Parachuting=(GravityOverride=()),OnGround_Standing=(GravityOverride=()),OnGround_Moving=(GravityOverride=()),OnGround_Moving_HighRotationRate=(GravityOverride=()),OnGround_Crouch_Moving=(GravityOverride=()),Emote_or_Melee=(GravityOverride=()),Swimming=(GravityOverride=()),Front_End=(GravityOverride=()),DBNO=(GravityOverride=()),DBNO_Carried=(GravityOverride=()),OnGround_Falling=(GravityOverride=()),Ziplining=(GravityOverride=()),Skydive_Surfing=(GravityOverride=()),Riding_Standing=(GravityOverride=()),Riding_Moving=(GravityOverride=()),Riding_Moving_Targeting=(GravityOverride=()),Tactical_Sprint=(GravityOverride=()),Rail_Grinding=(GravityOverride=()),)
 }
 GravityOverride {type=Vector, default=<0 0 -980>}
 Is playing StayAfloat {type=boolean}
]
